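# Material function that builds the glass look of the drivable Hamster Ball (OctoBall) vehicle, with damage cracks driven by a Health parameter.
# Crack normals fade in through FlattenNormal, one layer from full to half Health and a tiled layer below half, adding to the dome normal and emissive.
# Tinted, mip-biased skylight reflections add fake-reflection emissive, switched off on mobile.
# A glass sample supplies roughness, specular and opacity, with Fresnel boosting edge opacity.
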MaterialFunction /Game/Athena/DrivableVehicles/Hamster_Ball/Materials/MF_OctoBallMaterialLook
graph {nodes [
  n1 = TextureSampleParameter2D() {
   group=Vehicle_Textures
   parameterName=Diffuse
   samplerType=SAMPLERTYPE_Masks
   texture=/Game/Characters/Player/Female/Medium/Bodies/F_MED_Commando_01/Textures/SolidWhite_MaskCompression
  }
  n2 = Multiply(A=@n1.RGB, B=VectorParameter() {parameterName=Color, defaultValue=<rgba 0.257 0.3746 0.5 1>}.RGB)
  n3 = TextureSample() {texture=&/SampleGlass}
  n4 = TextureSample!(UVs=ReflectionVectorWS(), Tex, Bias=Multiply(A=@n3.B) {constB=3}, Apply View MipBias) {mipValueMode=TMVM_MipBias, samplerType=SAMPLERTYPE_LinearColor, texture=/Game/Packages/Fortress_Sky/TexturesHDR/T_AthenaSkylight}
  n5 = Multiply(A=@n2, B=@n4.RGB)
  n6 = Constant()
  n7 = Add(A=TextureCoordinate() {uTiling=5, vTiling=5}, B=Constant2Vector() {g=0.65, r=0.25})
  n8 = ScalarParameter() {defaultValue=1, parameterName=Health}
  n9 = MaterialFunctionCall(
   Normal=Multiply(
    A=TextureSample!(UVs=@n7, Tex, Apply View MipBias) {samplerType=SAMPLERTYPE_Normal, texture=/Game/Athena/DrivableVehicles/Hamster_Ball/Textures/T_Hamster_Ball_Glass_Cracks}.RGB
    B=Constant3Vector() {constant=<rgba 1 1 0 0>}
   )
   Flatness=Multiply(A=Clamp(None=@n8) {maxDefault=0.5}) {constB=2}
  ) {materialFunction=&FlattenNormal}
  n10 = Multiply(A=ComponentMask(None=@n9.Result) {g=true}) {constB=4}
  n11 = TextureCoordinate()
  n12 = Add(A=@n8) {constB=-0.5}
  n13 = MaterialFunctionCall(
   Normal=Multiply(
    A=TextureSample(UVs=@n11) {samplerType=SAMPLERTYPE_Normal, texture=/Game/Athena/DrivableVehicles/Hamster_Ball/Textures/T_Hamster_Ball_Glass_Cracks_02}.RGB
    B=Constant3Vector() {constant=<rgba 1 1 0 0>}
   )
   Flatness=Clamp(None=Multiply(A=@n12) {constB=2})
  ) {materialFunction=&FlattenNormal}
  n14 = ComponentMask(None=@n13.Result) {g=true}
  n15 = Multiply(A=Power(Base=Add(A=@n10, B=@n14)) {constExponent=1.5}, B=ScalarParameter() {defaultValue=0.5, parameterName=CrackEmissiveContribution})
  n16 = Multiply(A=Fresnel() {baseReflectFraction=0}, B=ScalarParameter() {defaultValue=5, parameterName=EdgeThicknessBoost})
  n17 = SetMaterialAttributes!(
   MaterialAttributes
   BaseColor=@n2
   Metallic=ScalarParameter() {defaultValue=1, parameterName=Metallic}
   Specular=@n3.R
   Roughness=@n3.R
   EmissiveColor=Add(A=ShadingPathSwitch(Default=@n5, Mobile=@n6), B=Clamp(None=@n15))
   Opacity=LinearInterpolate(A=@n3.G, Alpha=Clamp(None=@n16))
   Normal=Add(
    A=TextureSample() {samplerType=SAMPLERTYPE_Normal, texture=/Game/Athena/DrivableVehicles/Hamster_Ball/Textures/T_Hamster_Dome_N}.RGB
    B=Add(A=@n9.Result, B=@n13.Result)
   )
  ) {attributeSetTypes [
    #1
    #2
    #3
    #4
    #5
    #6
    #7
   ]}
  FunctionOutput(None=@n17)
 ]}
