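# M_OctopusBall is a translucent, lit surface material for the Hamster Ball drivable vehicle.
# Its graph holds a single call to MF_OctoBallMaterialLook, which contains the look logic and works through material attributes.
# It uses per-pixel surface translucency lighting, high-quality forward reflections and responsive AA, and supports skeletal meshes and mesh deformers.
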Material /Game/Athena/DrivableVehicles/Hamster_Ball/Materials/M_OctopusBall
graph {
 properties {
  bEnableResponsiveAA=true
  bUseHQForwardReflections=true
  bUseMaterialAttributes=true
  bUsedWithMeshDeformer=true
  bUsedWithSkeletalMesh=true
  blendMode=BLEND_Translucent
  pixelDepthOffsetMode=PDOM_Legacy
  translucencyLightingMode=TLM_SurfacePerPixelLighting
  usedShadingModels=Default Lit
 }
 nodes [MaterialFunctionCall() {materialFunction=&/MF_OctoBallMaterialLook}]
}
