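# Physics asset for the drivable Hamster Ball vehicle, covering four back tube bones and the inner ring, apparently for secondary tube motion.
# Each tube has one capsule, mirrored left to right, and inner_ring has two zero-length capsules that act as spheres.
# All four tubes constrain directly to inner_ring with limited swing and twist, the bk pair more tightly than the fr pair.
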PhysicsAsset /Game/Athena/DrivableVehicles/Hamster_Ball/meshes/Hamster_Ball_Veh_Dyn
bodies {
 back_tubes_bk_l [{shapeType=Capsule, length=22.8842, radius=2.214, center=<13.545 0 0>}]
 back_tubes_bk_r [{shapeType=Capsule, length=22.9467, radius=2.214, center=<-13.545 0 0>}]
 back_tubes_fr_l [{shapeType=Capsule, length=12.6378, radius=4.2532, center=<12.9146 0 0>}]
 back_tubes_fr_r [{shapeType=Capsule, length=12.6378, radius=4.2532, center=<-12.9146 0 0>}]
 inner_ring [
  {shapeType=Capsule, length=0, radius=13.5312, center=<-39.7639 -21.7175 -45.0864>, rotation=<rot -34.9997 0 15.0015>}
  {shapeType=Capsule, length=0, radius=13.5312, center=<-37.2457 22.1063 -44.5609>, rotation=<rot -34.9992 0 -24.9949>}
 ]
}
constraints [
 {bone1Name=back_tubes_fr_l, bone2Name=inner_ring, swing1LimitDegrees=20, swing2LimitDegrees=20, twistLimitDegrees=45}
 {bone1Name=back_tubes_bk_l, bone2Name=inner_ring, swing1LimitDegrees=20, swing2LimitDegrees=15, twistLimitDegrees=10}
 {bone1Name=back_tubes_fr_r, bone2Name=inner_ring, swing1LimitDegrees=20, swing2LimitDegrees=20, twistLimitDegrees=45}
 {bone1Name=back_tubes_bk_r, bone2Name=inner_ring, swing1LimitDegrees=20, swing2LimitDegrees=15, twistLimitDegrees=10}
]
